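; Passive quest ability that detects when the player holds a large enough stack (six, per its comments) of a qualifying splash or juice healing item.
; It listens for pickup and NPC service events, with a one-second delay letting new items reach the inventory.
; Checks gather items by required rarity to skip irrelevant items, filter by healing tags, then compare individual stack counts.
; Success sends the HoldingSixSplashOrJuice quest stat event.
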(event Ability|EventFailedToActivatePassiveAbility
  (params (ActorInfo :type Gameplay%20Ability%20Actor%20Info%20Structure))
  ;[1 1 1] Check if player has a weapon of any of the specified rarity types in inventory ~LS 3/20/23
  (|CallParentFunction ActorInfo))

(event Ability|EventActivateAbility
  ;[1 1 1] Check if player has a weapon of any of the specified rarity types in inventory ~LS 3/20/23
  (|CallParentFunction)
  (Ability|Tasks|WaitGameplayEvent "(TagName=\"Event.CollectedPickup\")"
    (:then
      (bind (_asynctaskproxy _payload) (Ability|Tasks|WaitGameplayEvent "(TagName=\"Event.NPCService\")" 0 false false)
        (:then
          ;;[1 1 1] Initial Inventory check when we get the quest.
          (Utilities|FlowControl|DoOnce false
            (:Completed
              _)))
        (:EventReceived
          (bind (_payload_eventtag _payload_instigator _payload_target _payload_optionalobject _payload_optionalobject2 _payload_contexthandle _payload_instigatortags) (Utilities|Struct|BreakGameplayEventData _payload))
          (if (GameplayTags|HasAnyTags _payload_instigatortags (Variables|Default|GetValidNPCServices) true)
            _)))
      (Utilities|Array|Clear (Variables|Default|GetPlayerItemsOfReqHealingType))
      (Utilities|Array|Clear (Variables|Default|GetItemsOfReqRarities))
      ;;[1 1 1] Delay a moment so the latest pickup item gets into the players inventory
      (Utilities|FlowControl|Delay 1.0)
      (Variables|Default|SetOwningPlayerInventoryInterface (Fortnite|GetFortPlayerControllerfromActor (Variables|GAVars|Data|GetOwningPlayerPawn)))
      (CallFunction|CheckforRareItems)
      (CallFunction|CheckSlapOrSlurp)
      (CallFunction|HasReqHealingItemAmount))
    (:EventReceived
      _)))

(event Custom|CheckForRareItems
  ;[1 1 1] Check if player has a weapon of any of the specified rarity types in inventory ~LS 3/20/23
  ;;[1 1 1] Check if there's any rare items. Doing this so it doesn't check against all the build tools and extra stuff as well
  (for _array_element (Variables|Default|GetRequiredRarityTypes)
    (bind _items (InventoryManagement|GetItemsfromInventoryOwnerforRarity (Variables|Default|GetOwningPlayerInventoryInterface) _array_element))
    (if (Utilities|Array|IsNotEmpty _items)
      (Utilities|Array|AppendArray (Variables|Default|GetItemsOfReqRarities) _items))))

(event Custom|Check%20Slap%20or%20Slurp
  ;[1 1 1] Check if player has a weapon of any of the specified rarity types in inventory ~LS 3/20/23
  ;;[1 1 1] Get all items of the specified healing items to one array
  (for _array_element (Variables|Default|GetItemsOfReqRarities)
    (if (GameplayTags|HasAnyTags (CallFunction|GetOwnedGameplayTags _array_element) (Variables|Default|GetT_RequiredHealingTags) true)
      (Utilities|Array|AddUnique (Variables|Default|GetPlayerItemsOfReqHealingType) _array_element))))

(event Custom|Has%20Req%20Healing%20Item%20Amount
  (bind _playeritemsofreqhealingtype (Variables|Default|GetPlayerItemsOfReqHealingType))
  ;[1 1 1] Check if player has a weapon of any of the specified rarity types in inventory ~LS 3/20/23
  ;;[1 1 1] Using the array of req healing items, check if there's 6 of them. If there is, give the player gameplay tag
  (if (Utilities|Array|IsNotEmpty _playeritemsofreqhealingtype)
    (bind _array_element (Utilities|Array|ForEachLoopwithBreak _playeritemsofreqhealingtype)
      (:LoopBody
        ;;[1 1 1] Do we have 6 in a stack? If yes, set bool to True and break the loop.
        (if (>= (Item|GetNuminStack _array_element) (Variables|Default|GetAmountToHold))
          (Variables|Default|SetSixSlapOrSlurpFound true))
        (Variables|Default|SetSixSlapOrSlurpFound false))
      (:Completed
        _))
    ;;[1 1 1] If bool is true at the end of the loop, give the player the CarryingVRPlusWeap Tag and end the ability
    (if (Variables|Default|GetSixSlapOrSlurpFound)
      (Quest|SendComplexCustomStatEvent (Variables|GAVars|Data|GetOwningPlayerQuestManager) 0 "(GameplayTags=((TagName=\"Athena.Quests.HoldingSixSplashOrJuice\")))" "(GameplayTags=)" 1)
      (Ability|EndAbility))))

(fn GetOwnedGameplayTags
  (params (self2 :type Gameplay%20Tag%20Asset%20Interface%20Interface))
  (return (GameplayTags|GetOwnedGameplayTags self2 :!class /Script/GameplayTags.GameplayTagAssetInterface)))
